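; Weapon component Blueprint that plays a one-time SoundLibrary "FirstUse" item sound the first time its owning weapon is used.
; On BeginPlay, where cosmetic events can run and AutoBind is set, it casts the owner to a Fort Weapon and binds to its trigger-press event.
; TriggerFirstAudio plays the tagged sound on the owner, sets a Fired flag to block repeats, and unbinds from the trigger.
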(event EventBeginPlay
  (Networking|CanExecuteCosmeticEvents
    (:True
      (Variables|Default|GetAutoBind
        (:then
          (bind _asfort_weapon (Utilities|Casting|CastToFortWeapon (Components|GetOwner)))
          (bind _output_get (Variables|Default|SetAs%20Fort%20Weapon _asfort_weapon))
          (EventDispatchers|BindEventtoOnPressTrigger _output_get (delegate Custom|FirstWeaponFire)))
        (:else)))))

(event Custom|FirstWeaponFire
  (params (Weapon :type Fort%20Weapon%20Object%20Reference))
  (CallFunction|TriggerFirstAudio))

(event Custom|TriggerFirstAudio
  (Variables|Default|GetFired
    (:then)
    (:else
      (SoundLibrary|PlaySoundSimple (WorldSubsystems|GetSoundLibrarySubsystem) (Components|GetOwner) (Utilities|Struct|MakeSimpleSoundSettings "(TagName=\"SoundLibrary.Item.FirstUse\")" true) :!class /Script/SoundLibrary.SoundLibrarySubsystem)
      (Variables|Default|SetFired true)
      (EventDispatchers|UnbindEventfromOnPressTrigger (Variables|Default|GetAs%20Fort%20Weapon) (delegate Custom|FirstWeaponFire)))))
