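; Weapon actor component that plays a looping Sound Library event while its weapon is equipped and visible, and stops it when holstered or hidden.
; Pawn and weapon delegates drive holster checks that also gate playback on vehicle state, with tag exemptions, and hiding in props.
; A gameplay tag query on the pawn sets an override for those play conditions.
; Helpers set bool and trigger parameters on the playing audio.
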(event EventBeginPlay
  (bind _playoverridetags (Variables|Event|GetPlayOverrideTags))
  (Variables|Private|SetWeapon (Components|GetOwner))
  ;[1 1 1] Bindings
  (CallFunction|WeaponBindings)
  (CallFunction|PlayerBindings)
  ;[1 1 1] Special tags what are watched for that will override all functionaility based on setup.
  (bind _playerpawn (Variables|Private|GetPlayerPawn)
    (:then
      (Ability|Tasks|WaitGameplayTagQueryOnActor _playerpawn _playoverridetags
        (:Triggered
          (Variables|Private|SetbOverridePlay true)
          (Ability|Tasks|WaitGameplayTagQueryOnActor (Variables|Private|GetPlayerPawn) _playoverridetags "WhenFalse" true
            (:Triggered
              (Variables|Private|SetbOverridePlay false))))))
    (:else)))

(event Custom|OnEquipped
  (params (NewWeapon :type Fort%20Weapon%20Object%20Reference) (PrevWeapon :type Fort%20Weapon%20Object%20Reference))
  ;[1 1 1] Weapon equip and unequip states. If player has this weapon selected whether it is visible or not.
  (if (== NewWeapon (Variables|Private|GetWeapon))
    (Variables|Private|SetbEquipped true)
    (CallFunction|OnUnHolestered)))

(event Custom|OnUnEquipped
  ;[1 1 1] Weapon equip and unequip states. If player has this weapon selected whether it is visible or not.
  (if (Variables|Private|GetbEquipped)
    (Variables|Private|SetbEquipped false)
    (CallFunction|OnHolestered)))

(event Custom|On%20Visibility%20Change
  (params (Visible :type Boolean))
  ;[1 1 1] Bind that updates if able based on weapons visiblity
  (bind _output_get (Variables|Private|SetbIsVisible Visible))
  (if _output_get
    (CallFunction|OnUnHolestered)
    (else
      (CallFunction|OnHolestered))))

(event Custom|OnVehicleStateChanged
  (params (FortPlayerPawn :type Fort%20Player%20Pawn%20Object%20Reference) (NewVehicle :type Actor%20Object%20Reference) (OldVehicle :type Actor%20Object%20Reference))
  ;[1 1 1] The state of weapon if player is in a vehicle with equipped weapon.
  (bind _output_get (Variables|Private|SetbIsInVehicle (and (Vehicle|IsinVehicle (Variables|Private|GetPlayerPawn)) (not (CallFunction|OverrideVehicleFuncions)))))
  (if (not _output_get)
    (if (not (Variables|Private|GetbIsVisible))
      (CallFunction|OnHolestered)))
  (CallFunction|OnUnHolestered))

(event Custom|OnUnHolestered
  ;[1 1 1] Holester/UnHolestered Events
  ;;[1 1 1] Chooses if audio should or not should play. Events are called from all other sources to perform a checks.
  (if (or (Math|Boolean|ANDBoolean (Variables|Private|GetbEquipped) (not (Variables|Private|GetbIsInVehicle)) (not (CallFunction|HidinginProp)) (Variables|Private|GetbIsVisible)) (Variables|Private|GetbOverridePlay))
    (CallFunction|PlayAudio)))

(event Custom|OnHolestered
  ;[1 1 1] Holester/UnHolestered Events
  ;;[1 1 1] Chooses if audio should or not should play. Events are called from all other sources to perform a checks.
  (CallFunction|StopAudio))

(event Custom|PlayAudio
  ;[1 1 1] Start/Stop Audio functions.
  (Variables|Private|GetSoundLibraryComponent
    (:then
      (if (not (Variables|Private|GetAudioPlaying))
        (CallFunction|UpdateLoopStatus true)
        (Variables|Private|SetAudioPlaying true)))
    (:else)))

(event Custom|StopAudio
  ;[1 1 1] Start/Stop Audio functions.
  (Variables|Private|GetSoundLibraryComponent
    (:then
      (if (Variables|Private|GetAudioPlaying)
        (CallFunction|UpdateLoopStatus false)
        (Variables|Private|SetAudioPlaying false)))
    (:else)))

(fn ExectuteBoolParameter
  (params (ParamName :type Name) (InBool :type Boolean))
  (bind _soundlibraryevent (Variables|Default|GetSoundLibraryEvent))
  (bind (_eventname _bwasblocked _bsuccess _audiocomponents) (Utilities|Struct|BreakSoundLibraryPlayResult _soundlibraryevent))
  (bind _output (Utilities|Array|Get(acopy) _audiocomponents))
  (Utilities|IsValid _output
    (:Is%20Valid
      (Audio|Parameter|SetBooleanParameter _output ParamName InBool :!class /Script/AudioExtensions.AudioParameterControllerInterface))))

(fn ExecuteTriggerParameter
  (params (ParamName :type Name))
  (bind _soundlibraryevent (Variables|Default|GetSoundLibraryEvent))
  (bind (_eventname _bwasblocked _bsuccess _audiocomponents) (Utilities|Struct|BreakSoundLibraryPlayResult _soundlibraryevent))
  (bind _output (Utilities|Array|Get(acopy) _audiocomponents))
  (Utilities|IsValid _output
    (:Is%20Valid
      (Audio|Parameter|ExecuteTriggerParameter _output ParamName))))

(fn GetWeaponOwner
  (return (Variables|Private|GetPlayerPawn)))

(fn HidingInProp
  (bind _bincludehideabletransition (Variables|Private|GetbIncludeHideableTransition))
  (bind _playerpawn (Variables|Private|GetPlayerPawn)
    (:then
      (Utilities|FlowControl|Sequence
        (:then_0
          (return (Class|FortPawn|IsInsideHidingProp _playerpawn _bincludehideabletransition)))
        (:then_1
          ;[1 1 1] Need to be updated after check is sent.
          (if (Class|FortPawn|IsInsideHidingProp _playerpawn _bincludehideabletransition)
            (Variables|Private|SetbIncludeHideableTransition false)
            (else
              (Variables|Private|SetbIncludeHideableTransition true))))))
    (:else)))

(fn OverrideVehicleFuncions
  ;[1 1 1] These vehicles have different functions than normal roofed vehicles, so we want to bypass in vehicle bool.
  (return (GameplayTags|HasAnyTags (GameplayTags|GetOwnedGameplayTags (Variables|Private|GetPlayerPawn) :!class /Script/GameplayTags.GameplayTagAssetInterface) (Variables|Vehicle|GetVehicleOverrideTags) true)))

(fn PlayValidEvent
  (params (GameplayTag :type Gameplay%20Tag%20Structure))
  (if (GameplayTags|IsGameplayTagValid GameplayTag)
    (bind _returnvalue (SoundLibrary|PlayEvent (Variables|Private|GetSoundLibraryComponent) GameplayTag))
    (return _returnvalue)))

(fn PlayerBindings
  (bind _asfort_player_pawn (Utilities|Casting|CastToFortPlayerPawn (Variables|/Script/Engine.Actor|GetInstigator (Variables|Private|GetWeapon))))
  (bind _output_get (Variables|Private|SetPlayerPawn _asfort_player_pawn))
  (EventDispatchers|BindEventtoOnHolsteredEvent _output_get (EventDispatchers|CreateEvent :!event OnHolestered))
  (EventDispatchers|BindEventtoOnUnholsteredEvent _output_get (EventDispatchers|CreateEvent :!event OnUnHolestered))
  (FortPawn|BindEventtoOnWeaponEquippedDelegate _output_get (EventDispatchers|CreateEvent :!event OnEquipped))
  (FortPawn|BindEventtoOnWeaponUnEquippedDelegate _output_get (EventDispatchers|CreateEvent :!event OnUnEquipped))
  (Vehicle|BindEventtoOnVehicleStateChangedEvent _output_get (EventDispatchers|CreateEvent :!event OnVehicleStateChanged)))

(fn UpdateLoopStatus
  (params (Play :type Boolean))
  (if Play
    (bind _result (CallFunction|PlayValidEvent (Variables|Event|GetLoopEventName)))
    (Variables|Default|SetSoundLibraryEvent _result)
    (else
      (bind _soundlibrarycomponent (Variables|Private|GetSoundLibraryComponent)
        (:then
          (if (SoundLibrary|IsEventActive _soundlibrarycomponent (Variables|Event|GetLoopEventName))
            (CallFunction|PlayValidEvent (Variables|Event|GetLoopingStopEventName))
            (SoundLibrary|StopEvent (Variables|Private|GetSoundLibraryComponent) (Variables|Event|GetLoopEventName))))
        (:else)))))

(fn WeaponBindings
  ;[1 1 1] Sound library component connects to soundlibrary weapon component.
  (bind _asfort_weapon (Utilities|Casting|CastToFortWeapon (Variables|Private|GetWeapon)))
  (Variables|Private|SetSoundLibraryComponent (SoundLibrary|GetSoundLibraryComponent _asfort_weapon "(TagName=\"SoundLibrary.Weapon\")" :!class /Script/FortniteGame.FortWeapon))
  (EventDispatchers|BindEventtoOnVisibilityChanged _asfort_weapon (EventDispatchers|CreateEvent :!event On%20Visibility%20Change)))
