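; Weapon audio component that plays overheat feedback for a Fortnite ranged weapon through Sound Library tags.
; Weapon fire, overheat, reload, trigger-release and visibility delegates drive heat sweetener, overheated/venting and cooldown-complete sounds.
; One-shot sweeteners are retriggered and loops updated by timer, with curves mapping heat percentage to pitch and volume.
; Sweetener and cooldown-complete sounds require a locally viewed pawn and visible weapon; hiding the weapon fades active components.
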(event EventBeginPlay
  ;[0 0.15 0.1488 0.5] Optimized
  ;; Initialize Component
  (CallFunction|EventandComponentBindings)
  (bind _asfort_player_pawn (Utilities|Casting|CastToFortPlayerPawn (Variables|/Script/Engine.Actor|GetInstigator (Variables|Default|GetWeapon))))
  (Variables|Default|SetPlayerPawn _asfort_player_pawn))

(event Custom|Weapon%20Fired
  (params (Weapon :type Fort%20Weapon%20Object%20Reference))
  ;[0 0.15 0.1488 0.5] Optimized
  (if (CallFunction|DetermineSoundShouldPlay)
    (if (Variables|Default|GetLoop)
      (CallFunction|StartPersistantFireOverheatLoop) ; Support for looping sources, e.g. Minigun
      (else
        (bind _sweetener_audio_component (Variables|Default|GetSweetener%20Audio%20Component)
          (:then
            (Audio|Parameter|ExecuteTriggerParameter _sweetener_audio_component "Retrigger"))
          (:else
            (bind _output (CallFunction|PlaySoundLibraryEventwithComponent "(TagName=\"SoundLibrary.Weapon.Overheat.Started\")"))
            (Variables|Default|SetSweetener%20Audio%20Component _output)))
        (CallFunction|UpdateParameters)))))

(event Custom|On%20Persistent%20Fire%20Shot
  (params (CurrentValue :type Float%20%28single-precision%29) (MaxValue :type Float%20%28single-precision%29) (State :type EFortWeaponOverheatState%20Enum) (bIsInCooldown :type Boolean))
  ;[0 0.15 0.1488 0.5] Optimized
  ;; Plays a sound each time the weapon overheat value increases for persistent FX weapons
  (switch Utilities|FlowControl|Switch|SwitchonEFortWeaponOverheatState State
    (:Heating
      (bind _sweetener_audio_component (Variables|Default|GetSweetener%20Audio%20Component)
        (:then
          (Audio|Parameter|ExecuteTriggerParameter _sweetener_audio_component "Retrigger")
          (CallFunction|UpdateParameters))
        (:else)))))

(event Custom|Update%20Parameters
  ;[0 0.15 0.1488 0.5] Optimized
  (bind _weapon (Variables|Default|GetWeapon)
    (:then
      (Variables|Default|SetOverheat%20Percentage (Math|Float|SafeDivide (Overheat|GetCurrentOverheatValue _weapon :!class /Script/FortniteGame.FortWeaponRanged) (Overheat|GetOverheatingMaxValue _weapon :!class /Script/FortniteGame.FortWeaponRanged)))
      (bind _overheataudiopitchfeedbackcurve (Variables|Default|GetOverheatAudioPitchFeedbackCurve)
        (:then
          (bind _sweetener_audio_component (Variables|Default|GetSweetener%20Audio%20Component)
            (:then
              (Audio|Parameter|SetFloatParameter _sweetener_audio_component "OverheatPitch" (+ (Math|Curves|GetFloatValue _overheataudiopitchfeedbackcurve (Variables|Default|GetOverheat%20Percentage)) (Variables|Default|GetSweetenerStartingPitch)) :!class /Script/AudioExtensions.AudioParameterControllerInterface)
              (bind _overheataudiovolumefeedbackcurve (Variables|Default|GetOverHeatAudioVolumeFeedbackCurve)
                (:then
                  (Audio|Parameter|SetFloatParameter (Variables|Default|GetSweetener%20Audio%20Component) "OverheatVolume" (Math|Curves|GetFloatValue _overheataudiovolumefeedbackcurve (Variables|Default|GetOverheat%20Percentage)) :!class /Script/AudioExtensions.AudioParameterControllerInterface))
                (:else)))
            (:else)))
        (:else)))
    (:else)))

(event Custom|StartPersistantFireOverheatLoop
  ;[0.0317 0.02655 0.15 0.5] Automatic Fire (Looping Sound)
  (if (CallFunction|DetermineSoundShouldPlay)
    (bind _overheatsweetenerlayeraudiocomponent (Variables|Default|GetOverheatSweetenerLayerAudioComponent)
      (:then
        (if (Audio|Components|Audio|IsPlaying _overheatsweetenerlayeraudiocomponent)
          _
          (else
            _)))
      (:else
        _))
    (bind _output (CallFunction|PlaySoundLibraryEventwithComponent (Variables|Default|GetOverheatStartedTag)))
    (bind _output_get (Variables|Default|SetOverheatSweetenerLayerAudioComponent _output))
    (CallFunction|ResetVolumeAndPitchParameters _output_get)
    (bind _returnvalue (Utilities|Time|SetTimerbyEvent (EventDispatchers|CreateEvent :!event Custom|UpdateOverheatSFX) 0.1 true))
    (Variables|Default|SetAutoOverheatTimer _returnvalue)))

(event Custom|GenericRangedOnFireFX
  (params (Weapon :type Fort%20Weapon%20Object%20Reference))
  (bind _loop (Variables|Default|GetLoop))
  ; Sound that Will Get Louder as Weapon Gets Closer to Overheating
  (if (Effects|ShouldPlayPersistentFireFX (Variables|Default|GetWeapon))
    (if _loop
      (CallFunction|StartPersistantFireOverheatLoop)
      (else
        (CallFunction|PersistantOneShoteFireFX)))
    (elif _loop
      (CallFunction|HandleManualFireOverheatLoop)
      (else
        (CallFunction|ManualFireOverheatSweetener)))))

(event Custom|OverheatValueChanged
  (params (CurrentValue :type Float%20%28single-precision%29) (MaxValue :type Float%20%28single-precision%29) (State :type EFortWeaponOverheatState%20Enum) (bIsInCooldown :type Boolean))
  ;[0 0.15 0.1488 0.5] Optimized
  ;; WatchOverheatState
  (Networking|CanExecuteCosmeticEvents
    (:True
      (switch Utilities|FlowControl|Switch|SwitchonEFortWeaponOverheatState State
        (:None
          (if (Variables|Default|GetHasOverheated)
            (CallFunction|FinishedCooling)))
        (:Overheated
          (Variables|Default|SetHasOverheated true)
          (CallFunction|OnOverheated))))))

(event Custom|StopPersistantFireLoop
  ;[0.0317 0.02655 0.15 0.5] Automatic Fire (Looping Sound)
  (Utilities|Time|ClearandInvalidateTimerbyHandle (Variables|Default|GetAutoOverheatTimer))
  (bind _overheatsweetenerlayeraudiocomponent (Variables|Default|GetOverheatSweetenerLayerAudioComponent)
    (:then
      (if (Audio|Components|Audio|IsPlaying _overheatsweetenerlayeraudiocomponent)
        (Audio|Components|Audio|FadeOut _overheatsweetenerlayeraudiocomponent (Variables|Default|GetOverheat%20Fade%20Out%20Duration) 0.0 "Linear" :!class /Script/Engine.AudioComponent)))
    (:else)))

(event Custom|UpdateOverheatSFX
  ;[0.0317 0.02655 0.15 0.5] Automatic Fire (Looping Sound)
  (if (Variables|Default|GetHasOverheated)
    (Utilities|Time|ClearandInvalidateTimerbyHandle (Variables|Default|GetAutoOverheatTimer))
    (bind _overheatsweetenerlayeraudiocomponent (Variables|Default|GetOverheatSweetenerLayerAudioComponent)
      (:then
        (if (Audio|Components|Audio|IsPlaying _overheatsweetenerlayeraudiocomponent)
          (Audio|Components|Audio|FadeOut _overheatsweetenerlayeraudiocomponent (Variables|Default|GetOverheat%20Fade%20Out%20Duration) 0.0 "Linear" :!class /Script/Engine.AudioComponent)))
      (:else))
    (else
      (CallFunction|UpdateVolumeandPitch (CallFunction|GetOverheatPercentage)))))

(event Custom|OnTriggerReleased
  (params (Weapon :type Fort%20Weapon%20Object%20Reference))
  ; Stop LoopingFX
  (CallFunction|StopWeaponFireOverheatFX))

(event Custom|OnOverheated
  ;[0 0.15 0.1488 0.5] Optimized
  ;; Sound to Play When Weapon is Overheated
  (bind _output (CallFunction|PlaySoundLibraryEventwithComponent "(TagName=\"SoundLibrary.Weapon.Overheat.Complete\")"))
  (Variables|Default|SetOnOverheatedAudioComponent _output))

(event Custom|Manual%20Reload
  (params (ReloadTime :type Float%20%28single-precision%29) (ReloadType :type EFortWeaponReloadType%20Enum :default "ReloadWholeClip"))
  ;[0 0.15 0.1488 0.5] Optimized
  ;; Premature Venting
  (if (not (Variables|Default|GetHasOverheated))
    (CallFunction|OnOverheated)))

(event Custom|OnPersistantFireStopped
  ; Stop LoopingFX
  (CallFunction|StopWeaponFireOverheatFX))

(event Custom|PersistantOneShoteFireFX
  ;[0.0225 0.0371 0.15 0.5] Automatic Fire (One Shot Sound)
  (if (CallFunction|DetermineSoundShouldPlay)
    (bind _output (CallFunction|PlaySoundLibraryEventwithComponent (Variables|Default|GetOverheatStartedTag)))
    (bind _output_get (Variables|Default|SetOverheatSweetenerLayerAudioComponent _output))
    (CallFunction|ResetVolumeAndPitchParameters _output_get)
    (Variables|Default|GetOverheatSweetenerLayerAudioComponent
      (:then
        (bind _returnvalue (Utilities|Time|SetTimerbyEvent (EventDispatchers|CreateEvent :!event Custom|UpdateOneShotOverHeatSFX) (CallFunction|CalculateFiringRateinSeconds) true))
        (Variables|Default|SetAutoOverheatTimer _returnvalue))
      (:else))))

(event Custom|FinishedCooling
  ;[0 0.15 0.1488 0.5] Optimized
  ;; Sound to Play When Cooling is Finished
  (if (CallFunction|DetermineSoundShouldPlay)
    (bind _onoverheatedaudiocomponent (Variables|Default|GetOnOverheatedAudioComponent)
      (:then
        (Audio|Components|Audio|FadeOut _onoverheatedaudiocomponent (Variables|Default|GetOverheat%20Fade%20Out%20Duration) 0.0 "Linear" :!class /Script/Engine.AudioComponent))
      (:else
        _))
    (CallFunction|PlaySoundLibraryEventwithComponent "(TagName=\"SoundLibrary.Weapon.Cooldown.Complete\")")
    (Variables|Default|SetHasOverheated false)))

(event Custom|StopPersistantFireOneShot
  ;[0.0225 0.0371 0.15 0.5] Automatic Fire (One Shot Sound)
  (bind _overheatsweetenerlayeraudiocomponent (Variables|Default|GetOverheatSweetenerLayerAudioComponent)
    (:then
      (if (Audio|Components|Audio|IsPlaying _overheatsweetenerlayeraudiocomponent)
        (Audio|Components|Audio|FadeOut _overheatsweetenerlayeraudiocomponent (Variables|Default|GetOverheat%20Fade%20Out%20Duration) 0.0 "Linear" :!class /Script/Engine.AudioComponent)
        (Utilities|Time|ClearandInvalidateTimerbyHandle (Variables|Default|GetAutoOverheatTimer))))
    (:else)))

(event Custom|UpdateOneShotOverHeatSFX
  ;[0.0225 0.0371 0.15 0.5] Automatic Fire (One Shot Sound)
  (if (Variables|Default|GetHasOverheated)
    (bind _overheatsweetenerlayeraudiocomponent (Variables|Default|GetOverheatSweetenerLayerAudioComponent)
      (:then
        (if (Audio|Components|Audio|IsPlaying _overheatsweetenerlayeraudiocomponent)
          (Audio|Components|Audio|FadeOut _overheatsweetenerlayeraudiocomponent (Variables|Default|GetOverheat%20Fade%20Out%20Duration) 0.0 "Linear" :!class /Script/Engine.AudioComponent)
          (Utilities|Time|ClearandInvalidateTimerbyHandle (Variables|Default|GetAutoOverheatTimer))))
      (:else))
    (else
      (bind _overheatsweetenerlayeraudiocomponent_2 (Variables|Default|GetOverheatSweetenerLayerAudioComponent)
        (:then
          (Audio|Parameter|ExecuteTriggerParameter _overheatsweetenerlayeraudiocomponent_2 "Retrigger")
          (CallFunction|UpdateVolumeandPitch (CallFunction|GetOverheatPercentage)))
        (:else
          (bind _output (CallFunction|PlaySoundLibraryEventwithComponent (Variables|Default|GetOverheatStartedTag)))
          (Variables|Default|SetOverheatSweetenerLayerAudioComponent _output))))))

(event Custom|VisibilityChanged
  (params (bVisible :type Boolean))
  ;[0 0.15 0.1488 0.5] Optimized
  ;; ManageVisibiltyChanged
  (bind _output_get (Variables|Default|SetVisible bVisible))
  (if (not _output_get)
    (Utilities|FlowControl|Sequence
      (:then_0
        (bind _onoverheatedaudiocomponent (Variables|Default|GetOnOverheatedAudioComponent)
          (:then
            (Audio|Components|Audio|FadeOut _onoverheatedaudiocomponent 0.3 0.0 "Linear" :!class /Script/Engine.AudioComponent))
          (:else)))
      (:then_1
        (bind _sweetener_audio_component (Variables|Default|GetSweetener%20Audio%20Component)
          (:then
            (Audio|Components|Audio|FadeOut _sweetener_audio_component 0.3 0.0 "Linear" :!class /Script/Engine.AudioComponent))
          (:else)))
      (:then_2
        (bind _overheatsweetenerlayeraudiocomponent (Variables|Default|GetOverheatSweetenerLayerAudioComponent)
          (:then
            (Audio|Components|Audio|FadeOut _overheatsweetenerlayeraudiocomponent 0.3 0.0 "Linear" :!class /Script/Engine.AudioComponent))
          (:else))))))

(event Custom|ManualFireOverheatSweetener
  ;[0.15 0.03827 0.02966 0.5] Manual Fire (One Shot Sound)
  (if (CallFunction|DetermineSoundShouldPlay)
    (bind _overheatsweetenerlayeraudiocomponent (Variables|Default|GetOverheatSweetenerLayerAudioComponent)
      (:then
        (Audio|Parameter|ExecuteTriggerParameter _overheatsweetenerlayeraudiocomponent "Retrigger")
        (CallFunction|UpdateVolumeandPitch (CallFunction|GetOverheatPercentage)))
      (:else
        (bind _output (CallFunction|PlaySoundLibraryEventwithComponent (Variables|Default|GetOverheatStartedTag)))
        (bind _output_get (Variables|Default|SetOverheatSweetenerLayerAudioComponent _output))
        (CallFunction|ResetVolumeAndPitchParameters _output_get)))))

(event Custom|StartManualFireOverheatLooping
  ;[0.15 0.0294 0.04088 0.5] Manual Fire (Looping Sound)
  (if (CallFunction|DetermineSoundShouldPlay)
    (bind _overheatsweetenerlayeraudiocomponent (Variables|Default|GetOverheatSweetenerLayerAudioComponent)
      (:then
        (if (Audio|Components|Audio|IsPlaying _overheatsweetenerlayeraudiocomponent)
          _
          (else
            _)))
      (:else
        _))
    (bind _output (CallFunction|PlaySoundLibraryEventwithComponent (Variables|Default|GetOverheatStartedTag)))
    (bind _output_get (Variables|Default|SetOverheatSweetenerLayerAudioComponent _output))
    (CallFunction|ResetVolumeAndPitchParameters _output_get)))

(event Custom|StopManualOneShotLoop
  ;[0.15 0.0294 0.04088 0.5] Manual Fire (Looping Sound)
  (bind _overheatsweetenerlayeraudiocomponent (Variables|Default|GetOverheatSweetenerLayerAudioComponent)
    (:then
      (if (Audio|Components|Audio|IsPlaying _overheatsweetenerlayeraudiocomponent)
        (Audio|Components|Audio|FadeOut _overheatsweetenerlayeraudiocomponent (Variables|Default|GetOverheat%20Fade%20Out%20Duration) 0.0 "Linear" :!class /Script/Engine.AudioComponent)
        (Variables|Default|SetLoopStarted false)))
    (:else)))

(event Custom|UpdateManualOverheatLoop
  ;[0.15 0.0294 0.04088 0.5] Manual Fire (Looping Sound)
  (if (Variables|Default|GetHasOverheated)
    (bind _overheatsweetenerlayeraudiocomponent (Variables|Default|GetOverheatSweetenerLayerAudioComponent)
      (:then
        (if (Audio|Components|Audio|IsPlaying _overheatsweetenerlayeraudiocomponent)
          (Audio|Components|Audio|FadeOut _overheatsweetenerlayeraudiocomponent (Variables|Default|GetOverheat%20Fade%20Out%20Duration) 0.0 "Linear" :!class /Script/Engine.AudioComponent)
          (Variables|Default|SetLoopStarted false)))
      (:else))
    (else
      (CallFunction|UpdateVolumeandPitch (CallFunction|GetOverheatPercentage)))))

(fn CalculateFiringRateInSeconds
  (return (+ (/ 1.0 (Math|Float|Max(Float) (FortWeapon|GetFiringRate (Variables|Default|GetWeapon)) 0.1)) 0.0001)))

(fn ClearSweetenerAudioComponent
  (Variables|Default|SetSweetener%20Audio%20Component 0))

(fn DetermineSoundShouldPlay
  (bind _playerpawn (Variables|Default|GetPlayerPawn)
    (:then
      (return (and (Pawn|IsLocallyViewed _playerpawn) (Variables|Default|GetVisible))))
    (:else)))

(fn EventAndComponentBindings
  (bind _asfort_weapon_ranged (Utilities|Casting|CastToFortWeaponRanged (Components|GetOwner)))
  (bind _output_get (Variables|Default|SetWeapon _asfort_weapon_ranged))
  (EventDispatchers|BindEventtoOnReleaseTrigger _output_get (EventDispatchers|CreateEvent :!event OnTriggerReleased))
  (EventDispatchers|BindEventtoOnWeaponFireAudio _output_get (EventDispatchers|CreateEvent :!event Weapon%20Fired))
  (Overheating|BindEventtoOnOverheatValueChanged _output_get (EventDispatchers|CreateEvent :!event OverheatValueChanged))
  (EventDispatchers|BindEventtoOnVisibilityChanged _output_get (EventDispatchers|CreateEvent :!event VisibilityChanged))
  (EventDispatchers|BindEventtoOnReloadStarted _output_get (EventDispatchers|CreateEvent :!event Manual%20Reload))
  (if (Effects|ShouldPlayPersistentFireFX (Variables|Default|GetWeapon))
    (Overheating|BindEventtoOnOverheatValueChanged (Variables|Default|GetWeapon) (EventDispatchers|CreateEvent :!event On%20Persistent%20Fire%20Shot))))

(fn GetOverheatPercentage
  (return (/ (Overheat|GetCurrentOverheatValue (Variables|Default|GetWeapon) :!class /Script/FortniteGame.FortWeaponRanged) (Overheat|GetOverheatingMaxValue (Variables|Default|GetWeapon) :!class /Script/FortniteGame.FortWeaponRanged))))

(fn HandleManualFireOverheatLoop
  (if (Variables|Default|GetLoopStarted)
    (CallFunction|StartManualFireOverheatLooping)
    (Variables|Default|SetLoopStarted true)
    (else
      (CallFunction|UpdateManualOverheatLoop))))

(fn OnRep_HasOverheated)

(fn PlaySoundLibraryEventWithComponent
  (params (SoundLibraryEventName :type Gameplay%20Tag%20Structure))
  (bind _returnvalue (WorldSubsystems|GetSoundLibrarySubsystem))
  (Utilities|FlowControl|Sequence
    (:then_0
      (bind _weapon (Variables|Default|GetWeapon)
        (:then
          (Utilities|IsValid _returnvalue
            (:Is%20Valid
              (bind _returnvalue_2 (SoundLibrary|PlaySoundSimple _returnvalue _weapon (Utilities|Struct|MakeSimpleSoundSettings SoundLibraryEventName true false true false 1.0 1.0 0.0 0 0 "None" (Variables|/Script/Engine.Actor|GetRootComponent (Variables|Default|GetWeapon))) :!class /Script/SoundLibrary.SoundLibrarySubsystem))
              (bind (_returnvalue_eventname _returnvalue_bwasblocked _returnvalue_bsuccess _returnvalue_audiocomponents) (Utilities|Struct|BreakSoundLibraryPlayResult _returnvalue_2))
              (if (Utilities|Array|IsValidIndex _returnvalue_audiocomponents)
                (return (Utilities|Array|Get(acopy) _returnvalue_audiocomponents))))))
        (:else)))
    (:then_1
      (return 0))))

(fn ResetVolumeandPitchParameters
  (params (AudioComponent :type Audio%20Component%20Object%20Reference))
  (Utilities|IsValid AudioComponent
    (:Is%20Valid
      (Audio|Components|Audio|SetVolumeMultiplier AudioComponent 0.01 :!class /Script/Engine.AudioComponent)
      (Audio|Components|Audio|SetPitchMultiplier AudioComponent (Variables|Default|GetSweetenerStartingPitch))
      (return))))

(fn StopWeaponFireOverheatFX
  (Utilities|FlowControl|Sequence
    (:then_0
      (CallFunction|StopPersistantFireLoop))
    (:then_1
      (CallFunction|StopPersistantFireOneShot))
    (:then_2
      (CallFunction|StopManualOneShotLoop))))

(fn UpdateOverheatCurves
  (params (OverHeatAudioVolumeFeedbackCurve :type Curve%20Float%20Object%20Reference) (OverheatAudioPitchFeedbackCurve :type Curve%20Float%20Object%20Reference))
  (Variables|Default|SetOverHeatAudioVolumeFeedbackCurve OverHeatAudioVolumeFeedbackCurve)
  (Variables|Default|SetOverheatAudioPitchFeedbackCurve OverheatAudioPitchFeedbackCurve))

(fn UpdateVolumeAndPitch
  (params (OverheatPercentage :type Float%20%28single-precision%29))
  (bind _overheatsweetenerlayeraudiocomponent (Variables|Default|GetOverheatSweetenerLayerAudioComponent)
    (:then
      (Audio|Components|Audio|SetVolumeMultiplier _overheatsweetenerlayeraudiocomponent (Math|Curves|GetFloatValue (Variables|Default|GetOverHeatAudioVolumeFeedbackCurve) OverheatPercentage) :!class /Script/Engine.AudioComponent)
      (Audio|Components|Audio|SetPitchMultiplier _overheatsweetenerlayeraudiocomponent (+ (Math|Curves|GetFloatValue (Variables|Default|GetOverheatAudioPitchFeedbackCurve) OverheatPercentage) (Variables|Default|GetSweetenerStartingPitch))))
    (:else)))
